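# Niagara system for the Wasabi emote's "Snikt" effect, with timed blade meshes, glint flashes and a popping Snikt mesh on the character.
# Delayed blade and flash bursts each spawn two particles, using spawn groups to pick User.SkeletalMesh bones or sockets and toggle blade renderers.
# The Snikt mesh follows clavicle_r through its own skeletal interface, using scale and material curves plus late-ramped gravity.
# The separate right-side blade and flash emitters are disabled.
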NiagaraSystem /BRCosmetics/Effects/Fort_Effects/Emotes/Wasabi/NS_Wasabi_Snikt
dependencies {
 usedDataInterfaces [&NiagaraDataInterfaceVectorCurve, &NiagaraDataInterfaceCurve]
 usedDynamicInputs []
 usedModules [
  &self:SystemSpawnScript.SystemScriptSource.SystemScriptGraph.NiagaraNodeAssignment_0.SetVariables_#1
  &SystemState
 ]
}
emitters {
 Flash_L {
  emitterUpdateScript [
   &EmitterState(Life Cycle Mode=::NewEnumerator1, Loop Behavior=::NewEnumerator1, Loop Delay=0.65, Loop Duration=1, UseLoopDelay=true)
   &SpawnBurst_Instantaneous(Spawn Count=2, Use Spawn Probability=false)
  ]
  particleSpawnScript [
   /Niagara/Modules/Spawn/Initialization/InitializeParticle(
    Color=<rgba 0.6 0.6 0.6 1>
    Lifetime=0.3
    Module.Write Color=true
    Module.Write Lifetime=true
    Module.Write Mass=true
    Module.Write Material Random=true
    Module.Write Position=false
    Module.Write Scale=false
    Module.Write SpriteRotation=true
    Module.Write SpriteSize=true
    Position=&SimulationPosition()
    Sprite Rotation=&UniformRangedFloat(Minimum=0, Maximum=1)
    Sprite Size=&Vector2DFromFloat(Value=100)
   )
   &FN_SetSpawnGroup_v2(Number of Spawn Groups=2)
   /SetVariables(Particles.Position=&Add_Vector(A=Linked() {linkedVariable {name=Particles.Position, type=NiagaraPosition}}, B=<0 0 0>)) {enabled=false, isSetParameters=true}
  ]
  particleUpdateScript [
   &ParticleState()
   &SkeletalMeshLocation(
    Bone / Socket Index=$Particles.SpawnGroup
    Bone Sampling Mode=::NewEnumerator4
    Inherit Velocity=false
    Interpolate Mesh Sampling=false
    MaskSpawnGroup=false
    Sampled Position Offset=/Game/Effects/Niagara/DynamicInputs/Water/FN_PickVectorFromSpawnGroupCutoff(
     Under Cutoff Vector:Vector3f=<0 35 5>
     Over Cutoff Vector:Vector3f=<0 35 -5>
     SpawnGroup:NiagaraInt32=$Particles.SpawnGroup
     Cutoff:NiagaraInt32=1
    )
    Sampled Position Offset Coordinate Space:ENiagaraMeshTransforms=::NewEnumerator3
    Skeletal Mesh=$User.SkeletalMesh
    Use Sampled Mesh Attribute as Ribbon ID=false
    Write Sampled Mesh Attrributes to Particle Attributes=true
   )
   &SolveForcesAndVelocity(Clamp Velocity=false, Limit Acceleration=false)
   &Color(
    Color=<rgba 0.4306 0.6707 0.7188 1>
    Scale Color=&VectorFromFloat(Value=&FloatFromCurve(FloatCurve=DataInterface() {propertyValues {Curve {keys [<at 0 5>,<at 0.3 1>]}}}, Scale Curve=2))
   )
   &ScaleSpriteSize(Scale Factor=&Vector2DFromFloat(Value=&FloatFromCurve(FloatCurve=DataInterface() {propertyValues {Curve {keys [
          <at -0.0035 0.0142 RCIM_Cubic>
          {arriveTangent=0.6433, interpMode=RCIM_Cubic, leaveTangent=0.6433, time=0.2414, value=0.9767}
          {arriveTangent=-1.2875, interpMode=RCIM_Cubic, leaveTangent=-1.2875, time=0.481, value=0.3258}
          <at 1 0>
         ]}}})))
   &SpriteRotationRate(Delta Time:NiagaraFloat=$Particles.Lifetime, Rotation Rate=30)
   &CameraOffset(Camera Offset Amount=5)
  ]
  rendererClasses [&NiagaraSpriteRendererProperties]
  renderers [{rendererClass=&NiagaraSpriteRendererProperties}]
 }
 Flash_R {
  bEnabled=false
  emitterUpdateScript [
   &EmitterState(Life Cycle Mode=::NewEnumerator1, Loop Behavior=::NewEnumerator1, Loop Delay=0.4, Loop Duration=1, UseLoopDelay=true)
   &SpawnBurst_Instantaneous(Spawn Count=1, Use Spawn Probability=false)
  ]
  particleSpawnScript [
   &InitializeParticle(
    Color=<rgba 0.6 0.6 0.6 1>
    Lifetime=0.3
    Module.Write Color=true
    Module.Write Lifetime=true
    Module.Write Mass=true
    Module.Write Material Random=true
    Module.Write Position=false
    Module.Write Scale=false
    Module.Write SpriteRotation=true
    Module.Write SpriteSize=true
    Position=&SimulationPosition()
    Sprite Rotation=&UniformRangedFloat(Minimum=0, Maximum=1)
    Sprite Size=&Vector2DFromFloat(Value=100)
   )
   &FN_SetSpawnGroup_v2(Override SpawnGroup=true, SpawnGroup=1)
  ]
  particleUpdateScript [
   &ParticleState()
   &SkeletalMeshLocation(
    Apply Sampled Orientation to Mesh Particles=true
    Bone / Socket Index=$Particles.SpawnGroup
    Bone Sampling Mode=::NewEnumerator4
    Inherit Velocity=false
    Interpolate Mesh Sampling=false
    MaskSpawnGroup=false
    Sampled Position Offset=<-42 10 -3>
    Sampled Position Offset Coordinate Space:ENiagaraMeshTransforms=::NewEnumerator3
    Skeletal Mesh=$User.SkeletalMesh
    Use Sampled Mesh Attribute as Ribbon ID=false
    Write Sampled Mesh Attrributes to Particle Attributes=true
   )
   &SolveForcesAndVelocity(Clamp Velocity=false, Limit Acceleration=false)
   &Color(
    Color=<rgba 0.4306 0.6707 0.7188 1>
    Scale Color=&VectorFromFloat(Value=&FloatFromCurve(FloatCurve=DataInterface() {propertyValues {Curve {keys [<at 0 5>,<at 0.3 1>]}}}, Scale Curve=2))
   )
   &ScaleSpriteSize(Scale Factor=&Vector2DFromFloat(Value=&FloatFromCurve(FloatCurve=DataInterface() {propertyValues {Curve {keys [
          <at -0.0035 0.0142 RCIM_Cubic>
          {arriveTangent=0.6433, interpMode=RCIM_Cubic, leaveTangent=0.6433, time=0.2414, value=0.9767}
          {arriveTangent=-1.2875, interpMode=RCIM_Cubic, leaveTangent=-1.2875, time=0.481, value=0.3258}
          <at 1 0>
         ]}}})))
   &SpriteRotationRate(Delta Time:NiagaraFloat=$Particles.Lifetime, Rotation Rate=30)
   &CameraOffset(Camera Offset Amount=5)
  ]
  rendererClasses [&NiagaraSpriteRendererProperties]
  renderers [{rendererClass=&NiagaraSpriteRendererProperties}]
 }
 Sharp_blade_L {
  emitterUpdateScript [
   &EmitterState(Life Cycle Mode=::NewEnumerator1, Loop Behavior=::NewEnumerator1, Loop Delay=0.25, Loop Duration=1, UseLoopDelay=true)
   &SpawnBurst_Instantaneous(Spawn Count=2, Use Spawn Probability=false)
  ]
  particleSpawnScript [
   &FN_SetSpawnGroup_v2(Number of Spawn Groups=2)
   &InitializeParticle(
    Color=<rgba 0.4095 0.4698 0.6 1>
    Lifetime=0.4
    Mesh Scale=&Multiply_Vector(A=<1 1 1>, B=&VectorFromFloat(Value=1))
    Module.Write Color=true
    Module.Write Lifetime=true
    Module.Write Mass=true
    Module.Write Material Random=false
    Module.Write Position=false
    Module.Write Scale=true
    Module.Write SpriteRotation=false
    Module.Write SpriteSize=false
    Position=&SimulationPosition()
    Sprite Size=<75 75>
   )
   /SetVariables(Particles.Renderer_Visibility:NiagaraInt32=&MaskIntBySpawnGroup(Masked Int=0, Passthrough Int=1, Spawn Group Mask=0)) {isSetParameters=true}
   /SetVariables(Particles.Renderer2_Visibility:NiagaraInt32=&MaskIntBySpawnGroup(Masked Int=1, Passthrough Int=0, Spawn Group Mask=0)) {isSetParameters=true}
  ]
  particleUpdateScript [
   &ParticleState()
   &SkeletalMeshLocation(
    Apply Sampled Orientation to Mesh Particles=true
    Bone / Socket Index=$Particles.SpawnGroup
    Bone Sampling Mode=::NewEnumerator4
    Inherit Velocity=false
    Interpolate Mesh Sampling=false
    MaskSpawnGroup=false
    Sampled Position Offset Coordinate Space:ENiagaraMeshTransforms=::NewEnumerator3
    Skeletal Mesh=$User.SkeletalMesh
    Use Sampled Mesh Attribute as Ribbon ID=false
    Write Sampled Mesh Attrributes to Particle Attributes=true
   )
   /SetVariables(Particles.MeshOrientation=&MultiplyQuaternion(
     Quaternion A=$Particles.SkeletalMeshLocation.SampledRotation
     Quaternion B=&MakeQuatFromAxisAngle(
      Axis=&FN_PickVectorFromSpawnGroupCutoff(
       Under Cutoff Vector:Vector3f=<0 0 0>
       Over Cutoff Vector:Vector3f=<0 0 0>
       SpawnGroup:NiagaraInt32=$Particles.SpawnGroup
       Cutoff:NiagaraInt32=1
      )
      AngleInDegrees=/Game/Effects/Niagara/DynamicInputs/FN_PickFloatFromSpawnGroupCutoff(
       Over Cutoff Float:NiagaraFloat=-90
       Under Cutoff Float:NiagaraFloat=-90
       SpawnGroup:NiagaraInt32=$Particles.SpawnGroup
       Cutoff:NiagaraInt32=1
      )
     )
    )) {enabled=false, isSetParameters=true, module=SetVariables_#2}
   &SolveForcesAndVelocity(Clamp Velocity=false, Limit Acceleration=false)
   &DynamicMaterialParameters(
    Index 0 Param 1=&FloatFromCurve(FloatCurve=DataInterface() {propertyValues {Curve {keys [<at 0 0>,<at 1 1.7>]}}})
    Module.Param0WriteEnabled=true
    Module.Param1WriteEnabled=true
    Module.Param2WriteEnabled=true
    Module.Param3WriteEnabled=true
   )
   &CameraOffset(Camera Offset Amount=10)
  ]
  rendererClasses [&NiagaraMeshRendererProperties]
  renderers [
   {rendererClass=&NiagaraMeshRendererProperties, rendererIndex=0}
   {rendererClass=&NiagaraMeshRendererProperties, rendererIndex=1}
  ]
 }
 Sharp_blade_R {
  bEnabled=false
  emitterUpdateScript [
   &EmitterState()
   &SpawnBurst_Instantaneous(Spawn Count=1, Use Spawn Probability=false)
  ]
  particleSpawnScript [
   &FN_SetSpawnGroup_v2(Override SpawnGroup=true, SpawnGroup=1)
   &InitializeParticle(
    Color=<rgba 0.4095 0.4698 0.6 1>
    Lifetime=0.4
    Mesh Scale=&Multiply_Vector(A=<1 1 1>, B=&VectorFromFloat(Value=1.25))
    Module.Write Color=true
    Module.Write Lifetime=true
    Module.Write Mass=true
    Module.Write Material Random=false
    Module.Write Position=false
    Module.Write Scale=true
    Module.Write SpriteRotation=false
    Module.Write SpriteSize=false
    Position=&SimulationPosition()
    Sprite Size=<200 200>
   )
  ]
  particleUpdateScript [
   &ParticleState()
   &SkeletalMeshLocation(
    Apply Sampled Orientation to Mesh Particles=true
    Bone / Socket Index=$Particles.SpawnGroup
    Bone Sampling Mode=::NewEnumerator4
    Inherit Velocity=false
    Interpolate Mesh Sampling=false
    MaskSpawnGroup=false
    Sampled Position Offset=<0 3.4 0>
    Sampled Position Offset Coordinate Space:ENiagaraMeshTransforms=::NewEnumerator3
    Skeletal Mesh=$User.SkeletalMesh
    Use Sampled Mesh Attribute as Ribbon ID=false
    Write Sampled Mesh Attrributes to Particle Attributes=true
   )
   /SetVariables(Particles.MeshOrientation=&MultiplyQuaternion(
     Quaternion A=$Particles.SkeletalMeshLocation.SampledRotation
     Quaternion B=&MakeQuatFromAxisAngle(Axis=<0 0 1>, AngleInDegrees=82)
    )) {isSetParameters=true, module=SetVariables_#2}
   &SolveForcesAndVelocity(Clamp Velocity=false, Limit Acceleration=false)
   &DynamicMaterialParameters(
    Index 0 Param 1=&FloatFromCurve(FloatCurve=DataInterface() {propertyValues {Curve {keys [<at 0 0>,<at 1 1.7>]}}})
    Module.Param0WriteEnabled=true
    Module.Param1WriteEnabled=true
    Module.Param2WriteEnabled=true
    Module.Param3WriteEnabled=true
   )
  ]
  rendererClasses [&NiagaraMeshRendererProperties]
  renderers [{rendererClass=&NiagaraMeshRendererProperties, rendererIndex=0}]
 }
 Snikt {
  emitterUpdateScript [
   &EmitterState()
   &SpawnBurst_Instantaneous(Spawn Count=1, Use Spawn Probability=false)
  ]
  particleSpawnScript [
   &FN_SetSpawnGroup_v2(Override SpawnGroup=true, SpawnGroup=1)
   &InitializeParticle(
    Lifetime=0.6
    Mesh Scale=&VectorFromFloat(Value=0.2)
    Module.Write Color=true
    Module.Write Lifetime=true
    Module.Write Mass=true
    Module.Write Material Random=true
    Module.Write Position=true
    Module.Write Scale=true
    Module.Write SpriteRotation=false
    Module.Write SpriteSize=false
    Position=&SimulationPosition()
    Sprite Size=<25 25>
   )
  ]
  particleUpdateScript [
   &ParticleState()
   &SkeletalMeshLocation(
    Bone / Socket Index=$Particles.SpawnGroup
    Bone Sampling Mode=::NewEnumerator4
    Inherit Velocity=false
    Interpolate Mesh Sampling=false
    MaskSpawnGroup=false
    Sampled Position Offset=<15 -40 0>
    Sampled Position Offset Coordinate Space:ENiagaraMeshTransforms=::NewEnumerator3
    Skeletal Mesh=$User.SkeletalMesh
    Use Sampled Mesh Attribute as Ribbon ID=false
    Write Sampled Mesh Attrributes to Particle Attributes=true
   ) {enabled=false}
   &SkeletalMeshLocation(
    Bone / Socket Index=$Particles.SpawnGroup
    Bone Sampling Mode=::NewEnumerator4
    Inherit Velocity=false
    Interpolate Mesh Sampling=false
    MaskSpawnGroup=false
    Sampled Position Offset=<-50 0 0>
    Sampled Position Offset Coordinate Space:ENiagaraMeshTransforms=::NewEnumerator3
    Skeletal Mesh=DataInterface() {propertyValues {
      DefaultMesh=null
      ExcludeBoneName=root
      OptionalComponentClass=null
      PreviewMesh=&M_Med_Soldier_04_Preview
      SkinningMode=SkinOnTheFly
      SoftSourceActor=null
      SourceMode=Default
      UvSetIndex=0
      WholeMeshLOD=-1
      bExcludeBone=false
      bReadDeformedGeometry=true
      bRequireCurrentFrameData=true
      ComponentTags []
      FilteredBones [clavicle_r]
      FilteredSockets []
      MeshUserParameter {parameter {name=None, typeDefHandle {registeredTypeIndex=1}, varData []}}
      SamplingRegions []
     }}
    Use Sampled Mesh Attribute as Ribbon ID=false
    Write Sampled Mesh Attrributes to Particle Attributes=true
   ) {module=SkeletalMeshLocation001}
   &ScaleMeshSize(Scale Factor=/Niagara/DynamicInputs/ScaleByCurve/ScaleVectorByCurve(
     InitialValue:Vector3f=<1 1 1>
     ScaleCurve:NiagaraDataInterfaceVectorCurve=DataInterface() {propertyValues {
       XCurve {keys [
         <at 0 0 RCIM_Cubic>
         {arriveTangent=2.2843, interpMode=RCIM_Cubic, leaveTangent=2.2843, time=0.3178, value=1.2}
         {arriveTangent=-0.4627, interpMode=RCIM_Cubic, leaveTangent=-0.4627, time=0.4378, value=1}
         {arriveTangent=-1.7786, interpMode=RCIM_Cubic, leaveTangent=-1.7786, time=0.75, value=1}
         <at 1 0 RCIM_Cubic>
        ]}
       YCurve {keys [
         <at 0 0 RCIM_Cubic>
         {arriveTangent=2.2886, interpMode=RCIM_Cubic, leaveTangent=2.2886, time=0.3167, value=1.4998}
         {arriveTangent=-1.1535, interpMode=RCIM_Cubic, leaveTangent=-1.1535, time=0.5159, value=1.1808}
         {arriveTangent=-2.4394, interpMode=RCIM_Cubic, leaveTangent=-2.4394, time=0.75, value=1}
         <at 1 0 RCIM_Cubic>
        ]}
       ZCurve {keys [
         <at 0 0 RCIM_Cubic>
         {arriveTangent=3.0427, interpMode=RCIM_Cubic, leaveTangent=3.0427, time=0.25, value=1.3258}
         {arriveTangent=-0.4, interpMode=RCIM_Cubic, leaveTangent=-0.4, time=0.37, value=1.1258}
         {arriveTangent=-1.787, interpMode=RCIM_Cubic, leaveTangent=-1.787, time=0.75, value=1.1258}
         <at 1 0 RCIM_Cubic>
        ]}
      }}
     CurveIndex:NiagaraFloat=$Particles.NormalizedAge
    ))
   &DynamicMaterialParameters(
    Index 0 Param 1=&FloatFromCurve(FloatCurve=DataInterface() {propertyValues {Curve {keys [
         {arriveTangent=3.8817, interpMode=RCIM_Cubic, leaveTangent=3.8817, tangentMode=RCTM_User, time=0, value=0}
         {arriveTangent=0.7221, interpMode=RCIM_Cubic, leaveTangent=0.7221, tangentMode=RCTM_User, time=0.7564, value=2.091}
         {arriveTangent=4.2941, interpMode=RCIM_Cubic, leaveTangent=4.2941, tangentMode=RCTM_User, time=0.9228, value=2.078}
        ]}}})
    Index 0 Param 2=1
    Index 0 Param 3=&FloatFromCurve(
     FloatCurve=DataInterface() {propertyValues {Curve {keys [
         {arriveTangent=-17.7515, interpMode=RCIM_Cubic, leaveTangent=-17.7515, tangentMode=RCTM_User, time=-0.0046, value=1.0113}
         {arriveTangent=-0.2845, interpMode=RCIM_Cubic, leaveTangent=-0.2845, tangentMode=RCTM_User, time=0.1, value=0.065}
         {arriveTangent=0.4619, interpMode=RCIM_Cubic, leaveTangent=0.4619, tangentMode=RCTM_User, time=0.6513, value=0.0607}
         {arriveTangent=3.5589, interpMode=RCIM_Cubic, leaveTangent=3.5589, tangentMode=RCTM_User, time=1, value=1}
        ]}}}
     Scale Curve=19
    )
    Module.Param0WriteEnabled=true
    Module.Param1WriteEnabled=true
    Module.Param2WriteEnabled=true
    Module.Param3WriteEnabled=true
   )
   &GravityForce(Gravity=&Lerp_Vector(
     A=<0 0 0>
     B=<0 0 -980>
     Alpha=&FloatFromCurve(FloatCurve=DataInterface() {propertyValues {Curve {keys [<at 0.65 0>,<at 0.675 1>]}}})
    ))
   &SolveForcesAndVelocity(Clamp Velocity=false, Limit Acceleration=false)
  ]
  rendererClasses [&NiagaraMeshRendererProperties]
  renderers [{rendererClass=&NiagaraMeshRendererProperties, rendererIndex=0}]
 }
}
summary {
 systemName=NS_Wasabi_Snikt
 emitters [
  {emitterName=Snikt, rendererClasses [&NiagaraMeshRendererProperties]}
  {emitterName=Sharp_blade_L, rendererClasses [&NiagaraMeshRendererProperties]}
  {bEnabled=false, emitterName=Sharp_blade_R, rendererClasses [&NiagaraMeshRendererProperties]}
  {emitterName=Flash_L, rendererClasses [&NiagaraSpriteRendererProperties]}
  {bEnabled=false, emitterName=Flash_R, rendererClasses [&NiagaraSpriteRendererProperties]}
 ]
 userVariables [User.SkeletalMesh {
   type=NiagaraDataInterfaceSkeletalMesh
   defaultValue {
    struct=NiagaraExt_VariableValue_DataInterface
    value {dataInterface=&self:NiagaraDataInterfaceSkeletalMesh_0}
   }
  }]
}
